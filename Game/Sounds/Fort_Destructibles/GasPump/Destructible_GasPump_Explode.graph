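# Sound cue for the explosion of a destructible gas pump, mixing close, medium and distant layers.
# Each layer randomly picks one of two waves with pitch variation and has its own attenuation and sound class override.
# The medium layer borrows grenade explosion waves and sits lower in the mix, while the close layer overrides its attenuation settings.
# A shared tank explosion stop-oldest concurrency rule is assigned to limit stacking.
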SoundCue /Game/Sounds/Fort_Destructibles/GasPump/Destructible_GasPump_Explode
root=Mixer_4
nodes [
 {
  id=Attenuation_0
  children [Modulator_2]
  properties {
   attenuationSettings=/Game/Sounds/Attenuation/Fort_Explosion_Distant_Athena_Att
   attenuationOverrides {falloffDistance=3000, attenuationShapeExtents {x=1000}}
  }
 }
 {
  id=Attenuation_4
  children [Random_0]
  properties {
   attenuationSettings=/Game/Sounds/Attenuation/Fort_Explosion_Close_Athena_Att
   bOverrideAttenuation=true
   attenuationOverrides {falloffDistance=4000, nonSpatializedRadiusStart=500, reverbWetLevelMin=0.2, stereoSpread=5000, attenuationShapeExtents {x=2000}}
  }
 }
 {
  id=Attenuation_5
  children [Modulator_5]
  properties {
   attenuationSettings=/Game/Sounds/Attenuation/Fort_Explosion_Medium_Athena_Att
   attenuationOverrides {falloffDistance=3000, attenuationShapeExtents {x=1000}}
  }
 }
 {id=Mixer_4, children [SoundClass_2,SoundClass_3,SoundClass_1], properties {inputVolume [1,0.75,1]}}
 {id=Modulator_1, children [Attenuation_4], properties {pitchMax=1.2, pitchMin=1, volumeMax=1, volumeMin=1}}
 {id=Modulator_2, children [Random_1], properties {pitchMax=1.1, pitchMin=0.9}}
 {id=Modulator_5, children [Random_3], properties {pitchMax=1.1, pitchMin=0.9}}
 {id=Random_0, children [WavePlayer_5,WavePlayer_4], properties {weights [1,1]}}
 {id=Random_1, children [WavePlayer_1,WavePlayer_3], properties {weights [1,1]}}
 {id=Random_3, children [WavePlayer_6,WavePlayer_7], properties {weights [1,1]}}
 {
  id=SoundClass_1
  children [Attenuation_0]
  properties {soundClassOverride=&Fort_WeaponRanged_Fire_Distant}
 }
 {id=SoundClass_2, children [Modulator_1], properties {soundClassOverride=/Game/Sounds/FortSoundClassesAndModes/Fort_Explosives_Large_PSM}}
 {
  id=SoundClass_3
  children [Attenuation_5]
  properties {soundClassOverride=&Fort_WeaponRanged_Fire_3P_PSM}
 }
 {id=WavePlayer_1, properties {soundWaveAssetPtr=&/Destructible_Gas_Pump_EXPLODE_Distant_01}}
 {id=WavePlayer_3, properties {soundWaveAssetPtr=&/Destructible_Gas_Pump_EXPLODE_Distant_02}}
 {id=WavePlayer_4, properties {soundWaveAssetPtr=&/Destructible_Gas_Pump_Exp_Close_02}}
 {id=WavePlayer_5, properties {soundWaveAssetPtr=&/Destructible_Gas_Pump_Exp_Close_01}}
 {id=WavePlayer_6, properties {soundWaveAssetPtr=/Game/Athena/Sounds/Weapons/Explosions/Grenade_Expl_Med_01}}
 {id=WavePlayer_7, properties {soundWaveAssetPtr=/Game/Athena/Sounds/Weapons/Explosions/Grenade_Expl_Med_02}}
]
properties {
 maxDistance=27500
 volumeMultiplier=0.9
 concurrencyOverrides {maxCount=2, volumeScale=0.7905}
 concurrencySet [/Game/Sounds/Concurrency_Rules/TankExplosion_StopOldest_Conc]
}
