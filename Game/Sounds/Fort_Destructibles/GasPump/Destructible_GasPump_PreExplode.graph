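# Sound cue for the gas pump destructible's pre-explosion sound.
# It randomly picks one of three equally weighted waves, routes it through a distant ranged-weapon-fire sound class override, and varies pitch upward and volume slightly.
# An attenuation node uses the RPG explosion settings with overrides for shape and distance-based low-pass filtering.
# The cue plays at reduced volume and overrides concurrency to limit simultaneous instances.
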SoundCue /Game/Sounds/Fort_Destructibles/GasPump/Destructible_GasPump_PreExplode
root=Attenuation_1
nodes [
 {
  id=Attenuation_1
  children [Modulator_0]
  properties {
   attenuationSettings=/Game/Sounds/Attenuation/Fort_Explosion_RPG_Att
   bOverrideAttenuation=true
   attenuationOverrides {bAttenuateWithLPF=true, lPFFrequencyAtMax=665.3513, lPFRadiusMax=2786.9099, lPFRadiusMin=2024.5316, attenuationShapeExtents {x=1407.6522}}
  }
 }
 {id=Modulator_0, children [SoundClass_0], properties {pitchMax=1.3, pitchMin=1, volumeMax=1, volumeMin=0.9}}
 {id=Random_1, children [WavePlayer_1,WavePlayer_2,WavePlayer_3], properties {weights [1,1,1]}}
 {id=SoundClass_0, children [Random_1], properties {soundClassOverride=&Fort_WeaponRanged_Fire_Distant}}
 {id=WavePlayer_1, properties {soundWaveAssetPtr=&/Destructible_Gas_Pump_PreExplode_01}}
 {id=WavePlayer_2, properties {soundWaveAssetPtr=&/Destructible_Gas_Pump_PreExplode_02}}
 {id=WavePlayer_3, properties {soundWaveAssetPtr=&/Destructible_Gas_Pump_PreExplode_03}}
]
properties {bOverrideConcurrency=true, maxDistance=5007.6523, volumeMultiplier=0.5, concurrencyOverrides {maxCount=2, volumeScale=0.8}}
